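# Translucent, unlit, two-sided master material for hologram effects on Niagara sprites, ribbons and meshes.
# Opacity is built from toggleable panning band, noise, local-aligned texture and gradient layers, plus Fresnel, detail texture, distortion, dissolve and depth fade.
# Particle color drives emissive with dissolve-edge highlights and time-of-day scaling, while dynamic parameters supply dissolve and layer offsets.
# Quality switches limit custom-depth backface fading to High/Epic and band/noise vertex displacement to Epic.
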Material /DealershipSharedLibrary/FX/Materials/MasterMaterials/Hologram/MM_Hologram
graph {
 properties {
  bAutomaticallySetUsageInEditor=false
  bUsedWithNiagaraMeshParticles=true
  bUsedWithNiagaraRibbons=true
  bUsedWithNiagaraSprites=true
  blendMode=BLEND_Translucent
  shadingModel=MSM_Unlit
  twoSided=true
  usedShadingModels=Unlit
 }
 nodes [
  n1 = ScalarParameter() {defaultValue=0.05, group=Panning Band, parameterName=Panning Band Width, sortPriority=3}
  n2 = ScalarParameter() {group=Panning Band, parameterName="Panning  Band Sharpness", sortPriority=3}
  n3 = Time()
  n4 = ScalarParameter() {defaultValue=0.1, group=Panning Band, parameterName="Panning  Band  Speed", sortPriority=4}
  n5 = Add(A=NamedRerouteUsage(), B=Sine(None=Multiply(A=@n3, B=@n4)))
  n6 = StaticBoolParameter() {group=Panning Band, parameterName=Panning Noise Direction, sortPriority=1}
  n7 = NamedRerouteUsage()
  n8 = ScalarParameter() {defaultValue=0.05, group=Panning Band, parameterName=Panning Band Distortion, sortPriority=4}
  n9 = Multiply(
   A=Add(A=TextureCoordinate(), B=Multiply(A=@n7, B=@n8))
   B=ScalarParameter() {defaultValue=2, group=Panning Band, parameterName="Panning  Band Tiling", sortPriority=2}
  )
  NamedRerouteDeclaration(None=StaticSwitchParameter(
    True=MaterialFunctionCall!(Width=@n1, Sharpness=@n2, Offset=@n5, Direction Switch=@n6, Compare, Input Coordinates=@n9) {materialFunction=/Engine/Functions/Engine_MaterialFunctions03/Procedurals/GeneratedBand}.Result
    False=Constant()
   ) {defaultValue=true, group=Panning Band, parameterName=Include Panning Band?, sortPriority=0}) {name=Panning Band, nodeColor=<rgba 0.1981 1 0.1878 1>}
  n10 = Multiply(
   A=TextureCoordinate()
   B=AppendVector(
    A=ScalarParameter() {defaultValue=0.5, group=Panning Noise, parameterName=Panning Noise Tiling X, sortPriority=2}
    B=ScalarParameter() {defaultValue=0.5, group=Panning Noise, parameterName=Panning Noise Tiling Y, sortPriority=2}
   )
  )
  n11 = Multiply(A=NamedRerouteUsage(), B=ScalarParameter() {defaultValue=0.2, group=Panning Noise, parameterName=Panning Noise Distortion, sortPriority=4})
  n12 = Constant()
  n13 = Add(
   A=NamedRerouteUsage()
   B=Multiply(A=Time(), B=ScalarParameter() {defaultValue=0.1, group=Panning Noise, parameterName="Panning  Noise  Speed", sortPriority=3})
  )
  n14 = AppendVector(A=@n12, B=@n13)
  n15 = AppendVector(A=@n13, B=@n12)
  n16 = ScalarParameter() {group=Panning Noise, parameterName="Panning  Noise  Direction", sortPriority=3}
  n17 = TextureSampleParameter2D(UVs=Add(A=Add(A=@n10, B=@n11), B=LinearInterpolate(A=@n14, B=@n15, Alpha=@n16))) {
   group=Panning Noise
   parameterName=Panning Noise Texture
   sortPriority=-1
   texture=/Game/Effects/Fort_Effects/Textures/Tile/T_GaussianNoise_01_M
  }
  n18 = LinearInterpolate(
   A=ChannelMaskParameter(None=@n17.RGBA) {group=Panning Noise, parameterName=Panning Noise Channel, sortPriority=-1}.R
   B=@n17.RGB
   Alpha=ScalarParameter() {group=Panning Noise, parameterName=Panning Noise Use Channel, sortPriority=-1}
  )
  n19 = ScalarParameter() {defaultValue=1.5, group=Panning Noise, parameterName="Panning  Noise  Power", sortPriority=0}
  NamedRerouteDeclaration(None=StaticSwitchParameter(True=Power(Base=@n18, Exp=@n19), False=Constant()) {defaultValue=true, group=Panning Noise, parameterName=Include Panning Noise?, sortPriority=-2}) {name=Panning Noise, nodeColor=<rgba 0.2666 1 0.9647 1>}
  n20 = TextureObjectParameter() {
   group=Local Aligned Texture
   parameterName=Local Aligned Texture
   sortPriority=-1
   texture=/Game/Effects/Fort_Effects/Textures/Tile/T_Hex_Noise
  }
  n21 = VectorParameter() {group=Local Aligned Texture, parameterName=Local Aligned Texture Size, sortPriority=1, defaultValue=<rgba 512 512 512 0>}
  n22 = LocalPosition()
  n23 = ScalarParameter() {defaultValue=5, group=Local Aligned Texture, parameterName=Local Aligned Texture Tiling, sortPriority=2}
  n24 = MaterialFunctionCall(TextureObject=@n20, TextureSize=@n21.RGB, WorldPosition=Add(A=Multiply(A=@n22.XYZ, B=@n23), B=NamedRerouteUsage())) {materialFunction=&WorldAlignedTexture}
  n25 = ScalarParameter() {defaultValue=0.1, group=Local Aligned Texture, parameterName=Local Aligned Texture Amount, sortPriority=1}
  NamedRerouteDeclaration(None=StaticSwitchParameter(True=Multiply(A=@n24.XY Texture, B=@n25), False=Constant()) {defaultValue=true, group=Local Aligned Texture, parameterName=Local Aligned Texture?, sortPriority=-2}) {name=Local Aligned Texture, nodeColor=<rgba 0.478 0.8532 1 1>}
  n26 = LinearInterpolate(A=ScalarParameter() {group=Backface Culling, parameterName=BackFace Opacity}, Alpha=TwoSidedSign())
  n27 = PixelDepth()
  n28 = ScalarParameter() {defaultValue=-1, group=Backface Culling, parameterName=BackFace Depth Bias}
  n29 = SceneTexture() {sceneTextureId=PPI_CustomDepth}
  n30 = ScalarParameter() {group=Backface Culling, parameterName=BackFace Opacity}
  n31 = Multiply(A=@n26, B=If(A=Add(A=@n27.R, B=@n28), B=ComponentMask(None=@n29.Color) {r=true}, "A > B"=@n30, "A == B"=@n30, "A < B"=Constant() {r=1}))
  NamedRerouteDeclaration(None=QualitySwitch(Default=@n26, Low=@n26, High=@n31, Medium=@n26, Epic=@n31)) {name=Backface Culling, nodeColor=<rgba 0.9377 0.4265 1 1>}
  n32 = ParticleColor()
  n33 = Add(A=NamedRerouteUsage(), B=NamedRerouteUsage())
  n34 = NamedRerouteUsage()
  n35 = NamedRerouteUsage()
  n36 = ScalarParameter() {defaultValue=0.15, group=Panning Band, parameterName="Panning Band  Amount"}
  n37 = Add(A=Add(A=Add(A=@n33, B=@n34), B=Multiply(A=@n35, B=@n36)), B=NamedRerouteUsage())
  n38 = Multiply(A=NamedRerouteUsage(), B=ScalarParameter() {defaultValue=0.2, group=Panning Noise, parameterName="Panning  Noise  Amount", sortPriority=0})
  n39 = Multiply(A=Multiply(A=Add(A=@n37, B=@n38), B=NamedRerouteUsage()), B=Saturate(None=NamedRerouteUsage()))
  NamedRerouteDeclaration(None=MaterialFunctionCall(Opacity In=Multiply(A=@n32.A, B=@n39)) {materialFunction=&MF_VCVFX_DepthFade}.Result) {name=Opacity, nodeColor=<rgba 0 0.6953 1 1>}
  n40 = ParticleColor()
  n41 = ScalarParameter() {defaultValue=75, group=Dissolve, parameterName=Dissolve Emissive Highlight}
  n42 = Multiply(
   A=LinearInterpolate(A=ParticleColor().RGB, B=Multiply(A=@n40.RGB, B=@n41) {constB=4096}, Alpha=NamedRerouteUsage())
   B=MaterialFunctionCall(
    Value_Morning=ScalarParameter() {defaultValue=1, group=Emissive, parameterName=Emissive Morning Time of Day}
    Value_Day=ScalarParameter() {defaultValue=1, group=Emissive, parameterName=Emissive Day Time of Day}
    Value_Evening=ScalarParameter() {defaultValue=1, group=Emissive, parameterName=Emissive Evening Time of Day}
    Value_Night=ScalarParameter() {defaultValue=1, group=Emissive, parameterName=Emissive Night Time of Day}
   ) {materialFunction=&Scalar_Time_of_Day}.Result
  )
  n43 = ScalarParameter() {defaultValue=1, group=Emissive, parameterName=Emissive Boost, sortPriority=1}
  NamedRerouteDeclaration(None=MaterialFunctionCall(Emissive=Multiply(A=@n42, B=@n43)) {materialFunction=&MF_FX_EmissiveAdjustments}.Result) {name=Emissive, nodeColor=<rgba 1 0.4726 0.8137 1>}
  n44 = Constant()
  n45 = LinearInterpolate(
   A=ScalarParameter() {defaultValue=-2, group=WPO, parameterName=WPO Min Amount}
   B=ScalarParameter() {defaultValue=5, group=WPO, parameterName="WPO Max  Amount"}
  ) {constA=-0.3, constB=0.3}
  n46 = Add(
   A=Multiply(A=NamedRerouteUsage(), B=ScalarParameter() {defaultValue=-2, group=WPO, parameterName=WPO Panning Band Amount}) {constB=-2}
   B=Multiply(A=NamedRerouteUsage(), B=ScalarParameter() {defaultValue=1, group=WPO, parameterName=WPO Panning Noise Amount}) {constB=-2}
  )
  n47 = Add(
   A=MaterialFunctionCall(Offset Amount=ScalarParameter() {group=WPO, parameterName=WPO CameraOffset}) {materialFunction=/Engine/Functions/Engine_MaterialFunctions02/WorldPositionOffset/CameraOffset}.Clamped Camera Offset
   B=Multiply(A=VertexNormalWS(), B=Multiply(A=@n45, B=@n46))
  )
  NamedRerouteDeclaration(None=Add(A=VertexTangentWS(), B=QualitySwitch(Default=@n44, Low=@n44, Medium=@n44, Epic=@n47))) {name=WPO, nodeColor=<rgba 0.6641 0 1 1>}
  n48 = LinearInterpolate(Alpha=NamedRerouteUsage()) {constB=1.5}
  n49 = Add(A=TextureCoordinate(), B=ScalarParameter() {group=Dissolve, parameterName=Dissolve Offset Randomization})
  n50 = AppendVector(
   A=ScalarParameter() {defaultValue=1, group=Dissolve, parameterName=Dissolve Tiling X, sortPriority=2}
   B=ScalarParameter() {defaultValue=1, group=Dissolve, parameterName=Dissolve Tiling Y, sortPriority=2}
  )
  n51 = TextureSampleParameter2D(UVs=Add(A=Multiply(A=@n49, B=@n50), B=NamedRerouteUsage())) {group=Dissolve, parameterName=Dissolve Texture, sortPriority=1, texture=/Game/Effects/Fort_Effects/Textures/T_Smooth_tile}
  NamedRerouteDeclaration(None=Subtract(A=@n48, B=@n51.RGB)) {name=Dissolve Opacity, nodeColor=<rgba 1 0.75 0 1>}
  n52 = Subtract(A=Add(A=@n51.RGB) {constB=0.45}, B=@n48)
  NamedRerouteDeclaration(None=Saturate(None=Ceil(None=@n52))) {name=Dissolve Emissive, nodeColor=<rgba 0.98 1 0.3824 1>}
  n53 = DynamicParameter() {parameterIndex=1, paramNames [Dissolve Amount, Dissolve Offset, Param3, Param4]}
  NamedRerouteDeclaration(None=@n53.Dissolve Amount) {name=DP_Dissolve Amount, nodeColor=<rgba 0 1 0.1094 1>}
  NamedRerouteDeclaration(None=@n53.Dissolve Offset) {name=DP_Dissolve Offset, nodeColor=<rgba 0 1 0.1094 1>}
  n54 = DynamicParameter() {paramNames [Panning Band Offset, Panning Noise Offset, Small Detail Offset, Local Aligned Texture Offset]}
  NamedRerouteDeclaration(None=@n54.Panning Band Offset) {name=DP_Panning Band Offset, nodeColor=<rgba 1 0.1957 0.2396 1>}
  NamedRerouteDeclaration(None=@n54.Panning Noise Offset) {name=DP_Panning Noise Offset, nodeColor=<rgba 1 0.1957 0.2396 1>}
  NamedRerouteDeclaration(None=@n54.Small Detail Offset) {name=DP_Small Detail Offset, nodeColor=<rgba 1 0.1957 0.2396 1>}
  NamedRerouteDeclaration(None=@n54.Local Aligned Texture Offset) {name=DP_Local Aligned Texture Offset, nodeColor=<rgba 1 0.1957 0.2396 1>}
  n55 = TextureCoordinate()
  n56 = VectorParameter() {
   group=Texture Distortion
   parameterName=Texture Distortion UV Controls
   sortPriority=2
   channelNames {a=Panner V, b=Panner U, g=Tiling V, r=Tiling U}
   defaultValue=<rgba 1 1 0.05 0.1>
  }
  n57 = AppendVector(A=@n56.Tiling U, B=@n56.Tiling U)
  n58 = TextureSampleParameter2D(UVs=Panner(Coordinate=Multiply(A=@n55, B=@n57), Speed=AppendVector(A=@n56.Panner U, B=@n56.Panner U))) {
   group=Texture Distortion
   parameterName=Texture Distortion - Texture
   sortPriority=1
   texture=/Game/Effects/Fort_Effects/Textures/Tile/T_PerlinRGB_01_M
  }
  NamedRerouteDeclaration(None=Subtract(A=ComponentMask(None=@n58.RGB) {g=true, r=true}) {constB=0.5}) {name=Texture Distortion Out, nodeColor=<rgba 0.8531 1 0.3891 1>}
  n59 = Fresnel(
   ExponentIn=ScalarParameter() {defaultValue=5, group=Fresnel, parameterName=Fresnel - Exponent, sortPriority=3}
   BaseReflectFractionIn=ScalarParameter() {defaultValue=0.04, group=Fresnel, parameterName=Fresnel - BaseReflectFraction, sortPriority=3}
  )
  n60 = Multiply(A=@n59, B=ScalarParameter() {defaultValue=0.2, group=Fresnel, parameterName=Fresnel - Amount, sortPriority=1})
  n61 = ScalarParameter() {group=Fresnel, parameterName=Fresnel - Bias, sortPriority=1}
  n62 = Multiply(
   A=Power(Base=@n59, Exp=ScalarParameter() {defaultValue=7, group=Fresnel, parameterName=Edge Highlight Thickness, sortPriority=2}) {constExponent=10}
   B=ScalarParameter() {defaultValue=50, group=Fresnel, parameterName=Edge Highlight Amount, sortPriority=2}
  ) {constB=50}
  NamedRerouteDeclaration(None=Add(A=Add(A=@n60, B=@n61), B=Saturate(None=@n62))) {name=Fresnel, nodeColor=<rgba 0.1981 1 0.1878 1>}
  n63 = Add(
   A=TextureCoordinate()
   B=Multiply(A=NamedRerouteUsage(), B=ScalarParameter() {group=Small Detail Texture, parameterName=Detail Texture Distortion, sortPriority=3})
  )
  n64 = AppendVector(
   A=ScalarParameter() {defaultValue=20, group=Small Detail Texture, parameterName=Detail- Tiling X, sortPriority=2}
   B=ScalarParameter() {defaultValue=20, group=Small Detail Texture, parameterName=Detail- Tiling Y, sortPriority=2}
  )
  n65 = Constant()
  n66 = NamedRerouteUsage()
  n67 = AppendVector(A=@n65, B=@n66)
  n68 = AppendVector(A=@n66, B=@n65)
  n69 = ScalarParameter() {group=Small Detail Texture, parameterName=Detail Offset Direction, sortPriority=4}
  n70 = TextureSampleParameter2D(UVs=Add(A=Multiply(A=@n63, B=@n64), B=LinearInterpolate(A=@n67, B=@n68, Alpha=@n69))) {
   group=Small Detail Texture
   parameterName=Detail Texture
   samplerType=SAMPLERTYPE_LinearColor
   sortPriority=0
   texture=/Game/Environments/Artemis/Props/Quest/Quest_TionfoilHoloTable_A/Textures/T_Tinfoil_HoloTable_Hexagons
  }
  NamedRerouteDeclaration(None=Multiply(A=@n70.RGB, B=ScalarParameter() {defaultValue=0.2, group=Small Detail Texture, parameterName=Detail - Strength, sortPriority=1})) {name=Detail Texture, nodeColor=<rgba 0.1981 1 0.1878 1>}
  n71 = LocalPosition()
  n72 = ScalarParameter() {defaultValue=2, group=Local Aligned Gradient, parameterName=Gradient Height, sortPriority=2}
  n73 = SphereMask(
   A=Divide(
    A=Add(A=@n71.XYZ, B=@n72)
    B=ScalarParameter() {defaultValue=0.5, group=Local Aligned Gradient, parameterName=Gradient Offset, sortPriority=2}
   )
   B=ObjectBounds()
  ) {hardnessPercent=0}
  n74 = Divide(
   A=LocalPosition().XYZ
   B=Multiply(A=ObjectBounds(), B=VectorParameter() {group=Local Aligned Gradient, parameterName=Linear Gradient Axis, defaultValue=<rgba 0 0 1 0>}.RGB)
  )
  n75 = Divide(
   A=Add(
    A=ComponentMask(None=@n74) {b=true}
    B=ScalarParameter() {defaultValue=0.5, group=Local Aligned Gradient, parameterName=Gradient Offset, sortPriority=2}
   )
   B=ScalarParameter() {defaultValue=2, group=Local Aligned Gradient, parameterName=Gradient Height, sortPriority=2}
  )
  n76 = Add(
   A=Power(
    Base=StaticSwitchParameter(True=@n73, False=@n75) {group=Local Aligned Gradient, parameterName=Sphere Gradient?, sortPriority=-2}
    Exp=ScalarParameter() {defaultValue=1, group=Local Aligned Gradient, parameterName=Gradient Power, sortPriority=0}
   )
   B=ScalarParameter() {defaultValue=-1, group=Local Aligned Gradient, parameterName=Gradient Bias, sortPriority=0}
  )
  NamedRerouteDeclaration(None=StaticSwitchParameter(True=Saturate(None=@n76), False=Constant()) {group=Local Aligned Gradient, parameterName=Local Aligned Gradient?, sortPriority=-3}) {name=Local Aligned Gradient, nodeColor=<rgba 0.5513 0.6388 1 1>}
  NamedRerouteUsage()
  NamedRerouteUsage()
  NamedRerouteUsage()
 ]
}
